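# Sound cue for the looping thrust audio of the GimmickFemale glider, using player glider thrust attenuation.
# It mixes two looping waves, a thrust loop and a move loop, each started after a short delay.
# Continuous modulators on both layers scale volume and pitch from the AnyDirection parameter, so each layer responds differently to that input.
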SoundCue /BRCosmetics/Sounds/Gliders/GimmickFemale/GliderThrustLoop_GImmickFemale_Cue
root=Mixer_0
nodes [
 {id=Delay_0, children [WavePlayer_0], properties {delayMax=0.5, delayMin=0.5}}
 {id=Delay_1, children [WavePlayer_1], properties {delayMax=0.6, delayMin=0.5}}
 {id=Mixer_0, children [ModulatorContinuous_2, ModulatorContinuous_3], properties {inputVolume [1,1]}}
 {
  id=ModulatorContinuous_2
  children [Delay_0]
  properties {
   pitchModulationParams {default=0, maxOutput=1.3, minOutput=1, parameterName=AnyDirection}
   volumeModulationParams {default=0, maxOutput=0.4, minOutput=0.2, parameterName=AnyDirection}
  }
 }
 {
  id=ModulatorContinuous_3
  children [Delay_1]
  properties {
   pitchModulationParams {default=0, minOutput=0.8, parameterName=AnyDirection}
   volumeModulationParams {default=0, maxInput=1.5, minOutput=0.1, parameterName=AnyDirection}
  }
 }
 {id=WavePlayer_0, properties {bLooping=true, soundWaveAssetPtr=&/Glider_GimmickFemale_Thrust_LP}}
 {id=WavePlayer_1, properties {bLooping=true, soundWaveAssetPtr=&/Glider_GimmickFemale_Thrust_LP_Move}}
]
properties {
 attenuationSettings=&Fort_Player_Glider_Thrust_Athena_Att
 audioPropertiesSheet=&APS_Cosmetic_Glider_Loop
 maxDistance=12500
 soundClassObject=&Fort_Gliders_Land
 volumeMultiplier=0.9
 concurrencySet [null]
}
